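# Conversation database for The Reaper NPC, defining its greeting and the branches that follow it.
# The entry point leads to one speech node with a default off-duty line and contextual lines for bar count, time since last match, Bullseye outfits and first meeting.
# Each contextual line carries a requirement and a curve-driven priority.
# The speech node then links to the shared NPC library quests and basic services conversations by entry tag.
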ConversationDatabase /NPCLibrary/NPCs/TheReaper/TheReaper/NPCLibrary_Conversation_TheReaper
entry_points [{tag=Conversation.Entry.TheReaper, destinations [#1]}]
nodes [
 ConversationEntryPointNode_7 {
  id=#1
  class=ConversationEntryPointNode
  outputs [#2]
  properties {entryTag {tagName=Conversation.Entry.TheReaper}}
 }
 Speech_C_4 {
  id=#2
  class=Speech_C
  outputs [#3, #4]
  properties {
   nodeName=""
   generalConfig {
    defaultMessage="You're <Hot>lucky</> I'm off-duty."
    contextualMessages [
     {
      message="Judging by the state of your wallet, I'd say work has been <b>going well.</>Nicely done."
      contextRequirements [{requirement {class=FortControllerRequirement_HasBars, properties {thresholdTestConfig {requirement=GreaterThanOrEqual}}}}]
      priority {curve {rowName=Tandem.Speech.Priority.HighTotalBars}}
     }
     {
      message="Took some time off I see? <b>I respect that.</>"
      contextRequirements [{requirement {class=FortControllerRequirement_CheckTimeSinceLastPlayedMatch}}]
      priority {curve {rowName=Tandem.Speech.Priority.WhileSincePlayed}}
     }
     {
      message=Too easy.
      requirementMatchPolicy=RequireAny
      contextRequirements [{requirement {
         class=FortControllerRequirement_HasCID
         properties {allowedCIds [
           /BRCosmetics/Athena/Items/Cosmetics/Characters/CID_242_Athena_Commando_F_Bullseye
           /BRCosmetics/Athena/Items/Cosmetics/Characters/CID_522_Athena_Commando_M_Bullseye
          ]}
        }}]
      priority {curve {rowName=Tandem.Speech.Priority.OutfitReaction}}
     }
     {
      message=You better not be muscling in on my turf.
      contextRequirements [{requirement {class=FortControllerRequirement_IsFirstEverConversationWithNPC}}]
      priority {curve {rowName=Tandem.Speech.Priority.FirstEverInteraction}}
     }
    ]
   }
  }
 }
 ConversationLinkNode_0 {id=#4, class=ConversationLinkNode, properties {remoteEntryTag {tagName=Conversation.Entry.BasicServices}}}
 ConversationLinkNode_1 {id=#3, class=ConversationLinkNode, properties {remoteEntryTag {tagName=Conversation.Entry.NPCLibrary.Quests}}}
]
